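# Post-process AnimBlueprint for the BitterSweet large male head cosmetic, built around head-bone rotation correction.
# On update, it negates the head's rotation delta from reference pose relative to neck_02, scales each axis by percentageEmote variables, and stores it in headRotLimit.
# The AnimGraph runs the input pose through an additive bone-space Modify Bone on head.
# That node shows zero rotation, with no visible link to headRotLimit.
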AnimBlueprint /BRCosmetics/Characters/Player/Male/Large/Heads/M_LRG_BitterSweet_Head/Meshes/M_LRG_BitterSweet_Head_AnimBP_Post
target_skeleton=/Game/Characters/Player/Common/Fortnite_Base_Head/Fortnite_Base_Head_Export_Skeleton
graphs [
 AnimGraph {nodes [
   n1 = "Transform (Modify) Bone - Bone: head"(ComponentPose=Local To Component(LocalPose=Input Pose()), Alpha=1, Translation="0,0,0", Rotation="0,0,0", Scale="1,1,1") {
    rotationMode=BMM_Additive
    rotationSpace=BCS_BoneSpace
    translationMode=BMM_Additive
    alphaBoolBlend {blendOption=Linear}
    boneToModify {boneName=head}
   }
   Output Pose(Result=Component To Local(ComponentPose=@n1))
  ]}
 EventGraph {nodes [
   Event BlueprintUpdateAnimation()
   n1 = GetOwnerSkeletalMeshComponent()
   n2 = BreakRotator(InRot=BreakTransform(InTransform=GetDeltaTransformFromRefPose(self=@n1, BoneName=head, BaseName=neck_02)).Rotation)
   n3 = Get Z percentageEmote() -> (Z percentageEmote:float)
   n4 = Get X percentageEmote() -> (X percentageEmote:float)
   n5 = Get Y percentageEmote() -> (Y percentageEmote:float)
   > Set headRotLimit(headRotLimit:Rotator=MakeRotator(
     Roll=float * float!(A=@n2, B=-1, C:double=@n4) -> (ReturnValue)
     Pitch=float * float(A=@n2.Pitch, B=-1, C=@n5)
     Yaw=float * float(A=@n2.Yaw, B=-1, C=@n3)
    )) -> (Output_Get:Rotator)
   TryGetPawnOwner()
  ]}
]
variables [
 headRotLimit {type=Rotator}
 X Percentage Emote {type=number}
 X percentageEmote {type=number, default=1}
 Y percentageEmote {type=number, default=1}
 Z percentageEmote {type=number, default=1}
]
